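# Unlit, translucent, two-sided trail material for the DarkEagleFire one-handed pickaxe, usable on beam trails, ribbons, sprites and mesh particles.
# Emissive blends two colors through panning noise, brightens the leading edge and can take particle color.
# Opacity combines edge masks, a curved sphere mask and ring noise stretched by lerping between two UV tilings.
# Switches and parameters enable plasma-cloud erosion driven by Dynamic Parameters, a bright leading edge, edge softening and depth fade.
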Material /BRCosmetics/Weapons/FORT_Melee/Pickaxe_DarkEagleFire_1h/FX/M_DarkEagleFire1h_Trail
graph {
 properties {
  bUsedWithBeamTrails=true
  bUsedWithMeshParticles=true
  bUsedWithNiagaraMeshParticles=true
  bUsedWithNiagaraRibbons=true
  bUsedWithNiagaraSprites=true
  blendMode=BLEND_Translucent
  pixelDepthOffsetMode=PDOM_Legacy
  shadingModel=MSM_Unlit
  twoSided=true
  usedShadingModels=Unlit
 }
 nodes [
  n1 = ParticleColor()
  n2 = MaterialFunctionCall() {materialFunction=&LinearGradient}
  n3 = Multiply(A=Power(Base=OneMinus(None=@n2.UGradient), Exp=ScalarParameter() {defaultValue=5, parameterName=GradientContrast})) {constB=3}
  n4 = Power(
   Base=OneMinus(None=MaterialFunctionCall() {materialFunction=&LinearGradient}.UGradient)
   Exp=ScalarParameter() {defaultValue=3, parameterName=GradientContrast2}
  )
  n5 = TextureCoordinate()
  n6 = Constant() {r=10}
  n7 = Constant() {r=0.8}
  n8 = Add(
   A=OneMinus(None=ComponentMask(None=@n5) {r=true})
   B=OneMinus(None=MaterialFunctionCall!(UVs, Mult=@n6, Exp=@n7) {materialFunction=/Game/Effects/Fort_Effects/MaterialFunctions/MF_SquareEdgeMask}.LeftMask)
  )
  n9 = MaterialFunctionCall(Float2=TextureCoordinate() {uTiling=1.2, vTiling=1.5}) {materialFunction=&BreakOutFloat2Components}
  n10 = OneMinus(None=@n9.G)
  n11 = Add(A=@n9.R) {constA=0.25, constB=-0.3}
  n12 = MaterialFunctionCall(Float2=TextureCoordinate() {uTiling=12, vTiling=3}) {materialFunction=&BreakOutFloat2Components}
  n13 = OneMinus(None=@n12.G)
  n14 = Add(B=@n12.R) {constA=0.25}
  n15 = MaterialFunctionCall(Float2=TextureCoordinate()) {materialFunction=&BreakOutFloat2Components}
  n16 = TextureSample(UVs=LinearInterpolate(
    A=MaterialFunctionCall(X=@n10, Y=@n11) {materialFunction=&MakeFloat2}.Result
    B=MaterialFunctionCall(X=@n13, Y=@n14) {materialFunction=&MakeFloat2}.Result
    Alpha=@n15.G
   )) {
   samplerType=SAMPLERTYPE_LinearColor
   texture=/Game/CharClasses/Soldier/Traits/AdvancedTactics/Abilities/GoingCommando/Textures/T_GoingCommando_RingNoise_01
  }
  n17 = Multiply(A=@n16.R, B=Power(Base=OneMinus(None=@n15.G)))
  n18 = Constant() {r=1}
  n19 = TextureCoordinate() {uTiling=0.5}
  n20 = Constant2Vector() {g=0.82, r=0.85}
  n21 = Constant() {r=0.8}
  n22 = Constant() {r=0.3}
  n23 = Power(Base=OneMinus(None=SphereMask(A=@n19, B=@n20, Radius=@n21, Hardness=@n22))) {constExponent=5}
  n24 = MaterialFunctionCall(Mult=Constant() {r=3}, Exp=ScalarParameter() {defaultValue=1, parameterName=RightExp}) {materialFunction=&MF_SquareEdgeMask}
  n25 = MaterialFunctionCall(Mult=ScalarParameter() {defaultValue=3, parameterName=TopBotMult}, Exp=ScalarParameter() {defaultValue=1, parameterName=TopBotExp}) {materialFunction=&MF_SquareEdgeMask}
  n26 = Constant() {r=4}
  n27 = Constant() {r=2}
  n28 = Multiply(
   A=Multiply(A=Add(A=@n8, B=@n17), B=StaticSwitchParameter(True=@n18, False=@n23) {parameterName=TurnOffCurvedMask})
   B=Multiply(
    A=Multiply(A=@n24.RightMask, B=@n25.HorizMask)
    B=MaterialFunctionCall(Mult=@n26, Exp=@n27) {materialFunction=&MF_SquareEdgeMask}.LeftMask
   )
  )
  n29 = TextureCoordinate()
  n30 = AppendVector(A=ScalarParameter() {defaultValue=1, parameterName=FireNoiseX}, B=ScalarParameter() {defaultValue=2, parameterName=FireNoiseY})
  n31 = TextureSample(UVs=Panner(Coordinate=Multiply(A=@n29, B=@n30)) {speedX=-2}) {texture=/Game/Effects/Fort_Effects/Textures/T_VFX_PlasmaClouds}
  n32 = Power(
   Base=Clamp(None=OneMinus(None=@n31.G))
   Exp=DynamicParameter() {defaultValue=<rgba 1 0 0 8>, paramNames [EdgeFade, Erode, ErodeMaskUVOffset, ErodeNoise]}.ErodeNoise
  )
  n33 = OneMinus(None=Clamp(None=Multiply(A=@n32)))
  n34 = ComponentMask(None=TextureCoordinate()) {r=true}
  n35 = Power(Base=Sine(None=Divide(A=@n34)))
  n36 = StaticSwitchParameter(
   True=LinearInterpolate(A=Subtract(A=@n28, B=@n33), B=Subtract(A=@n35, B=@n33), Alpha=ScalarParameter() {parameterName=0ForCubeMask, sliderMax=1})
   False=@n28
  ) {parameterName=UseErodeOnU}
  n37 = TextureCoordinate()
  n38 = Sine(None=Divide(A=ComponentMask(None=@n37) {g=true}))
  n39 = Multiply(
   A=Add(A=@n36, B=Add(A=@n3, B=@n4))
   B=LinearInterpolate(A=Power(Base=@n38), B=Constant() {r=1}, Alpha=ScalarParameter() {parameterName=0ForGSineMask, sliderMax=1})
  )
  n40 = Multiply(A=@n1.A, B=StaticSwitchParameter(True=Clamp(None=@n39), False=@n36) {parameterName=UseBrightLeadingEdge})
  n41 = TextureCoordinate()
  n42 = Sine(None=Divide(A=ComponentMask(None=@n41) {r=true}))
  n43 = LinearInterpolate(
   A=Constant() {r=3}
   B=DynamicParameter() {defaultValue=<rgba 1 0 0 8>, paramNames [EdgeFade, Erode, ErodeMaskUVOffset, ErodeNoise]}.EdgeFade
   Alpha=ScalarParameter() {parameterName=UseDynParamEdge}
  )
  n44 = LinearInterpolate(
   A=@n40
   B=Multiply(A=@n40, B=Power(Base=@n42, Exp=@n43) {constExponent=0})
   Alpha=ScalarParameter() {parameterName=1ToTakeOffHardEdgeOnR, sliderMax=1}
  )
  LinearInterpolate(
   A=@n44
   B=DepthFade(Opacity=@n44, FadeDistance=ScalarParameter() {defaultValue=50, parameterName=DepthFade})
   Alpha=ScalarParameter() {parameterName=UseDepthFade}
  )
  n45 = Power(
   Base=OneMinus(None=MaterialFunctionCall() {materialFunction=&LinearGradient}.UGradient)
   Exp=ScalarParameter() {defaultValue=5, parameterName=GradientContrast}
  )
  n46 = ScalarParameter() {defaultValue=3, parameterName=LeadingEdgeBrightness}
  n47 = Multiply(A=TextureCoordinate() {coordinateIndex=1}, B=ScalarParameter() {defaultValue=0.3, parameterName=Tile})
  n48 = Multiply(
   A=TextureCoordinate() {coordinateIndex=1}
   B=AppendVector(A=ScalarParameter() {defaultValue=1, parameterName=TileX}, B=ScalarParameter() {defaultValue=1, parameterName=TileY})
  )
  n49 = ScalarParameter() {parameterName=UseIndividualValues for Tiling}
  n50 = Time()
  n51 = ScalarParameter() {defaultValue=1, parameterName=Speed}
  n52 = ScalarParameter() {defaultValue=-2.5, parameterName=PannerSpeedX}
  n53 = ScalarParameter() {parameterName=PannerSpeedY}
  n54 = TextureSampleParameter2D(UVs=Panner(Coordinate=LinearInterpolate(A=@n47, B=@n48, Alpha=@n49), Time=Multiply(A=@n50, B=@n51), Speed=AppendVector(A=@n52, B=@n53)) {speedX=-2.5}) {parameterName=NoiseB, texture=/Game/Global/Textures/Default/Noise/TankMoteNoise}
  n55 = LinearInterpolate(
   A=VectorParameter() {parameterName=Color1, defaultValue=<rgba 3 0 0 1>}.RGB
   B=VectorParameter() {parameterName=Color2, defaultValue=<rgba 5 1 0.1 1>}.RGB
   Alpha=Multiply(A=@n8, B=@n54.G)
  )
  n56 = Add(A=Multiply(A=Multiply(A=@n45, B=@n46) {constB=3}, B=@n55), B=@n55)
  n57 = OneMinus(None=ComponentMask(None=TextureCoordinate()) {r=true})
  n58 = Add(A=Power(Base=@n57) {constExponent=10}, B=Divide(A=@n57))
  n59 = LinearInterpolate(A=@n56, B=Multiply(A=Clamp(None=@n58) {maxDefault=1.5}, B=@n56), Alpha=ScalarParameter() {parameterName=UseBrightnessRange?, sliderMax=1})
  MaterialFunctionCall(Emissive=LinearInterpolate(A=@n59, B=Multiply(A=@n59, B=@n1.RGB), Alpha=ScalarParameter() {parameterName=ParticleColorInfluence, sliderMax=1})) {materialFunction=&MF_FX_EmissiveAdjustments}
  Multiply(A=@n55, B=@n54.G)
 ]
}
