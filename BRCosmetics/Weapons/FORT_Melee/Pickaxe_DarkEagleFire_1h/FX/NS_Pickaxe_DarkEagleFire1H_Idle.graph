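# Idle Niagara effect for the one-handed DarkEagleFire pickaxe, emitting rising embers and dark smoke.
# Two continuous sprite emitters spawn along a thin, locally offset box, with drag and upward acceleration.
# Embers use curve-driven HDR orange color that fades out; translucent black smoke grows, slows its spin and fades in and out, both with SubUV animation.
# Standard pickaxe user parameters are declared, but neither emitter's shown modules read them.
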NiagaraSystem /BRCosmetics/Weapons/FORT_Melee/Pickaxe_DarkEagleFire_1h/FX/NS_Pickaxe_DarkEagleFire1H_Idle
dependencies {
 usedDataInterfaces [
  &NiagaraDataInterfaceCurve
  &NiagaraDataInterfaceSpriteRendererInfo
  &NiagaraDataInterfaceColorCurve
  &NiagaraDataInterfaceVectorCurve
  &NiagaraDataInterfaceSocketReader
 ]
 usedDynamicInputs []
 usedModules [
  &self:SystemSpawnScript.SystemScriptSource.SystemScriptGraph.NiagaraNodeAssignment_1.SetVariables_#1
  &SystemState
 ]
}
emitters {
 NE_Embers {
  emitterUpdateScript [
   &EmitterState()
   &SpawnRate(SpawnRate=3, Use Spawn Probability=false)
  ]
  particleSpawnScript [
   &InitializeParticle(
    Color=<rgba 5 0.25 0 1>
    Lifetime Min=0.5
    Lifetime Mode=::NewEnumerator1
    Position Offset=<0 0 30>
    Position Offset Coordinate Space=::Local
    Sprite Rotation Mode=::NewEnumerator2
    Sprite Size Mode=::NewEnumerator1
    Uniform Sprite Size Max=3.25
    Uniform Sprite Size Min=1.75
    UsePositionOffset=true
    Write Material Random=true
   )
   &ShapeLocation(Box Size=<50 1 1>, MaskSpawnGroup=false, Override Seed=false, Shape Primitive=::NewEnumerator2, Surface Only=false, UseOwnerScale=true)
  ]
  particleUpdateScript [
   &ParticleState()
   &ScaleSpriteSize(Uniform Curve Sprite Scale=DataInterface() {propertyValues {Curve {keys [
        {arriveTangent=0.5516, interpMode=RCIM_Cubic, leaveTangent=0.5516, tangentMode=RCTM_User, time=0, value=0}
        {arriveTangent=5.476, interpMode=RCIM_Cubic, leaveTangent=-0.224, tangentMode=RCTM_Break, time=0.4001, value=1.0058}
        {arriveTangent=0.0706, interpMode=RCIM_Cubic, leaveTangent=-23.6527, tangentMode=RCTM_Break, time=0.9106, value=1}
        <at 1 0>
       ]}}})
   &SpriteRotationRate(Rotation Rate=&AngleConversion(
     Angle:NiagaraFloat=&RandomRangeFloat(Minimum=-2, Maximum=2)
     Angle Input:ENiagara_AngleInput=::NewEnumerator1
     Angle Output:ENiagara_AngleInput=::NewEnumerator0
    ))
   &Drag(Drag=1.5, Use Linear Drag=true, Use Rotational Drag=true)
   &AccelerationForce(Acceleration=<0 0 80>)
   &SolveForcesAndVelocity(Clamp Velocity=false, Limit Acceleration=false)
   &SubUVAnimation(
    Random Start Frame=true
    "Sprite Renderer "=DataInterface() {propertyValues {SpriteRenderer=&self:NE_Embers_0.NiagaraSpriteRendererProperties_0}}
    UseEndFrame=false
    UseStartFrame=false
   )
   &ScaleColor(
    Scale RGB=&VectorFromCurve(VectorCurve=DataInterface() {propertyValues {
       XCurve {keys [<at 0 500 RCIM_Cubic>, {arriveTangent=-500, interpMode=RCIM_Cubic, leaveTangent=-500, time=0.68, value=10}, <at 1 0 RCIM_Cubic>]}
       YCurve {keys [
         <at 0 62.9696 RCIM_Cubic>
         {arriveTangent=-62.9696, interpMode=RCIM_Cubic, leaveTangent=-62.9696, time=0.68, value=0.426}
         <at 1 0 RCIM_Cubic>
        ]}
       ZCurve {keys [<at 0 0 RCIM_Cubic>,<at 0.68 0 RCIM_Cubic>,<at 1 0 RCIM_Cubic>]}
      }})
    ScaleA=true
    ScaleRGB=true
   )
  ]
  rendererClasses [&NiagaraSpriteRendererProperties]
  renderers [{rendererClass=&NiagaraSpriteRendererProperties}]
 }
 NE_Smoke {
  emitterUpdateScript [
   &EmitterState()
   &SpawnRate(SpawnRate=12, Use Spawn Probability=false)
  ]
  particleSpawnScript [
   &InitializeParticle(
    Color=<rgba 0 0 0 0.5>
    Lifetime Max=2.5
    Lifetime Min=1.5
    Lifetime Mode=::NewEnumerator1
    Position Offset=<0 0 30>
    Position Offset Coordinate Space=::Local
    Sprite Rotation Mode=::NewEnumerator2
    Sprite Size Mode=::NewEnumerator1
    Uniform Sprite Size Max=90
    Uniform Sprite Size Min=60
    UsePositionOffset=true
    Write Material Random=true
   )
   &AddVelocity(Module.Use Falloff Distance=false, Override Seed=false, Velocity Mode=::NewEnumerator1)
   &ShapeLocation(Box Size=<50 1 1>, MaskSpawnGroup=false, Override Seed=false, Shape Primitive=::NewEnumerator2, Surface Only=false, UseOwnerScale=true)
  ]
  particleUpdateScript [
   &ParticleState()
   &SpriteRotationRate(Rotation Rate=&AngleConversion(
     Angle:NiagaraFloat=&Multiply_Float(
      A=&FloatFromCurve(FloatCurve=DataInterface() {propertyValues {Curve {keys [
           {arriveTangent=-5.2264, interpMode=RCIM_Cubic, leaveTangent=-5.2264, tangentMode=RCTM_User, time=0, value=1}
           {arriveTangent=-0.2882, interpMode=RCIM_Cubic, leaveTangent=-0.2882, tangentMode=RCTM_User, time=0.4089, value=0.0544}
           <at 1 0>
          ]}}})
      B=&RandomRangeFloat(Minimum=-1, Maximum=1)
     )
     Angle Input:ENiagara_AngleInput=::NewEnumerator1
     Angle Output:ENiagara_AngleInput=::NewEnumerator0
    ))
   &ScaleSpriteSize(Uniform Curve Sprite Scale=DataInterface() {propertyValues {Curve {keys [<at 0 0.3 RCIM_Cubic>, {arriveTangent=0.6766, leaveTangent=0.6766, time=1, value=0.75}]}}})
   &SubUVAnimation(
    Evaluation Type=::NewEnumerator0
    "Sprite Renderer "=DataInterface() {propertyValues {SpriteRenderer=&self:NE_Smoke_0.NiagaraSpriteRendererProperties_0}}
    SubUV Animation Mode:ENiagara_SubUVLookupModeV2=::NewEnumerator2
    UseEndFrame=false
    UseStartFrame=false
   )
   &Drag(Drag=&RandomRangeFloat(Minimum=3, Maximum=6), Use Linear Drag=true, Use Rotational Drag=true)
   &AccelerationForce(Acceleration=&RandomRangeVector(Minimum=<0 0 30>, Maximum=<0 0 60>))
   &SolveForcesAndVelocity(Clamp Velocity=false, Limit Acceleration=false)
   &ScaleColor(
    Scale Alpha=&FloatFromCurve(FloatCurve=DataInterface() {propertyValues {Curve {keys [
         {arriveTangent=-0.005, interpMode=RCIM_Cubic, leaveTangent=-0.005, tangentMode=RCTM_User, time=0.0039, value=0.0063}
         {arriveTangent=0.504, interpMode=RCIM_Cubic, leaveTangent=0.504, time=0.3975, value=0.7684}
         {arriveTangent=-1.9454, interpMode=RCIM_Cubic, leaveTangent=-1.9454, tangentMode=RCTM_User, time=0.5896, value=0.3015}
         <at 1 0 RCIM_Cubic>
        ]}}})
    ScaleA=true
    ScaleRGB=true
   )
  ]
  rendererClasses [&NiagaraSpriteRendererProperties]
  renderers [{rendererClass=&NiagaraSpriteRendererProperties}]
 }
}
summary {
 systemName=NS_Pickaxe_DarkEagleFire1H_Idle
 emitters [
  {emitterName=NE_Embers, rendererClasses [&NiagaraSpriteRendererProperties]}
  {emitterName=NE_Smoke, rendererClasses [&NiagaraSpriteRendererProperties]}
 ]
 userVariables [
  User.bIsFrontEnd {type=NiagaraBool}
  User.bIsFrontEndPreview {type=NiagaraBool}
  User.bisSwinging {type=NiagaraBool}
  User.IsFlying {type=NiagaraFloat, defaultValue=0, description=""}
  User.IsSkydiving {type=NiagaraBool}
  User.PhysicalSurface {type=NiagaraInt32}
  User.RibbonIDOffset {type=NiagaraInt32}
  User.SwingSectionIndex {type=NiagaraInt32}
  User.VariantID {type=NiagaraFloat, defaultValue=0, description=""}
 ]
}
