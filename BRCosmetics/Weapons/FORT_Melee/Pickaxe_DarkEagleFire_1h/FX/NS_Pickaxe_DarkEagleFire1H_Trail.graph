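# Niagara trail effect for the one-handed DarkEagleFire pickaxe, combining a swing ribbon with embers.
# NE_Swing uses FN anim-trail modules to build a short-lived ribbon between two bones or sockets, with curves fading alpha and sharply dimming its bright color.
# NE_Embers continuously spawns small orange sprites in a local offset cylinder, with slight inverse inherited velocity.
# The embers shrink over their brief lifetimes while an HDR color curve dims their glow.
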NiagaraSystem /BRCosmetics/Weapons/FORT_Melee/Pickaxe_DarkEagleFire_1h/FX/NS_Pickaxe_DarkEagleFire1H_Trail
dependencies {
 usedDataInterfaces [
  &NiagaraDataInterfaceVectorCurve
  &NiagaraDataInterfaceCurve
  &NiagaraDataInterfaceColorCurve
  &NiagaraDataInterfaceSocketReader
 ]
 usedDynamicInputs []
 usedModules [
  &SystemState
  &self:SystemSpawnScript.SystemScriptSource.SystemScriptGraph.NiagaraNodeAssignment_1.SetVariables_#1
 ]
}
emitters {
 NE_Embers {
  emitterUpdateScript [
   &EmitterState()
   &SpawnRate(SpawnRate=50, Use Spawn Probability=false)
  ]
  particleSpawnScript [
   &InitializeParticle(
    Color=<rgba 5 0.25 0 1>
    Lifetime Max=0.3
    Lifetime Min=0.15
    Lifetime Mode=::NewEnumerator1
    Position Offset=<0 0 30>
    Position Offset Coordinate Space=::Local
    Sprite Rotation Mode=::NewEnumerator2
    Sprite Size Mode=::NewEnumerator1
    Uniform Sprite Size Max=5
    Uniform Sprite Size Min=3
    UsePositionOffset=true
    Write Material Random=true
   )
   &ShapeLocation(
    Cylinder Height=50
    Cylinder Radius=5
    MaskSpawnGroup=false
    Offset=<0 0 20>
    Offset Coordinate Space=::Local
    Offset Mode=::NewEnumerator0
    Override Seed=false
    Shape Primitive=::NewEnumerator1
    Surface Only=false
    UseOwnerScale=true
   )
   &InheritVelocity(Inherited Velocity Amount Scale=<-0.05 -0.05 -0.05>, Speed Limit=false, UseSpeedThreshold=false)
  ]
  particleUpdateScript [
   &ParticleState()
   &ScaleSpriteSize(Uniform Curve Sprite Scale=DataInterface() {propertyValues {Curve {defaultValue=1, keys [{tangentMode=RCTM_User, time=0, value=1}, {tangentMode=RCTM_User, time=1, value=0}]}}})
   &SolveForcesAndVelocity(Clamp Velocity=false, Limit Acceleration=false)
   &ScaleColor(
    Scale RGB=&VectorFromCurve(VectorCurve=DataInterface() {propertyValues {
       XCurve {keys [<at 0 500 RCIM_Cubic>, {arriveTangent=-500, interpMode=RCIM_Cubic, leaveTangent=-500, time=0.68, value=10}, <at 1 0 RCIM_Cubic>]}
       YCurve {keys [
         <at 0 62.9696 RCIM_Cubic>
         {arriveTangent=-62.9696, interpMode=RCIM_Cubic, leaveTangent=-62.9696, time=0.68, value=0.426}
         <at 1 0 RCIM_Cubic>
        ]}
       ZCurve {keys [<at 0 0 RCIM_Cubic>,<at 0.68 0 RCIM_Cubic>,<at 1 0 RCIM_Cubic>]}
      }})
    ScaleA=true
    ScaleRGB=true
   )
  ]
  rendererClasses [&NiagaraSpriteRendererProperties]
  renderers [{rendererClass=&NiagaraSpriteRendererProperties}]
 }
 NE_Swing {
  emitterSpawnScript [&FN_RibbonIDOffsetUtility()]
  emitterUpdateScript [
   &EmitterState()
   &FN_SpawnAnimTrail()
  ]
  particleSpawnScript [
   &FN_InitializeAnimTrailBoneIndecies(First Bone or Socket Index=1, Second Bone or Socket Index=0)
   &FN_InitializeAnimTrail(Color=<rgba 0.2 1.3 0 0.2>, Lifetime=0.175)
  ]
  particleUpdateScript [
   &SolveForcesAndVelocity(Clamp Velocity=false, Limit Acceleration=false)
   &FN_UpdateAnimTrail() {module=FN_UpdateAnimTrail001}
   &ScaleColor(
    Scale Alpha=&FloatFromCurve(FloatCurve=DataInterface() {propertyValues {Curve {keys [{interpMode=RCIM_Cubic, tangentMode=RCTM_User, time=0, value=1}, {arriveTangent=-2.9197, tangentMode=RCTM_User, time=0.5, value=0}]}}})
    Scale RGB=&VectorFromCurve(VectorCurve=DataInterface() {propertyValues {XCurve {keys [<at 0 200>,<at 1 1>]}, YCurve {keys [<at 0 200>,<at 1 1>]}, ZCurve {keys [<at 0 200>,<at 1 1>]}}})
    ScaleA=true
    ScaleRGB=true
   )
  ]
  rendererClasses [&NiagaraRibbonRendererProperties]
  renderers [{rendererClass=&NiagaraRibbonRendererProperties}]
 }
}
summary {
 systemName=NS_Pickaxe_DarkEagleFire1H_Trail
 emitters [
  {emitterName=NE_Swing, rendererClasses [&NiagaraRibbonRendererProperties]}
  {emitterName=NE_Embers, rendererClasses [&NiagaraSpriteRendererProperties]}
 ]
 userVariables [
  User.bIsFrontEnd {type=NiagaraBool}
  User.bIsFrontEndPreview {type=NiagaraBool}
  User.bisSwinging {type=NiagaraBool}
  User.PhysicalSurface {type=NiagaraInt32}
  User.RibbonIDOffset {type=NiagaraInt32}
  User.SwingSectionIndex {type=NiagaraInt32}
 ]
}
